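# Translucent, unlit, two-sided material for a TacoSoap melee trail on Niagara mesh particles, named as the cheap variant.
# Emissive blends parameter colors with a particle-tinted water wave color through banded UV gradients and panning plasma noise.
# A Perlin-noise and sphere-mask dissolve shapes the trail, with optional head sharpening, and opacity uses depth fade.
# Dynamic parameters drive erosion, UV growth and normal-based vertex offset thickness, optionally scaled by object size.
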Material /Game/Effects/Fort_Effects/GameplayPulginsAssets/TacoSoap/M_TacoSoap_MeleeTrail_Cheap
graph {
 properties {
  bAutomaticallySetUsageInEditor=false
  bUsedWithNiagaraMeshParticles=true
  blendMode=BLEND_Translucent
  previewMesh=/Engine/Transient.MaterialEditorMeshComponent_2
  shadingModel=MSM_Unlit
  twoSided=true
  usedShadingModels=Unlit
 }
 nodes [
  n1 = VectorParameter() {parameterName=Param_1, defaultValue=<rgba 0 0.1184 1 1>}
  n2 = Multiply(
   A=VectorParameter() {parameterName=WaterWaveColor, defaultValue=<rgba 0 0.937 1 1>}.RGB
   B=Multiply(A=Constant() {r=1.5}, B=ParticleColor().RGB)
  )
  n3 = OneMinus(None=MaterialFunctionCall() {materialFunction=&LinearGradient}.UGradient)
  n4 = ScalarParameter() {parameterName=2ColorGradientOffset}
  n5 = Power(Base=SmoothStep(Value=Add(A=@n3, B=@n4)) {constMin=0.5}, Exp=ScalarParameter() {defaultValue=1, parameterName=2ColorGradientPower})
  n6 = NamedRerouteUsage()
  n7 = MaterialFunctionCall(UV Channel=@n6) {materialFunction=&LinearGradient}
  n8 = Sine(None=Multiply(A=@n7.VGradient) {constB=0.5})
  n9 = LinearInterpolate(B=@n8, Alpha=Floor(None=Multiply(A=@n7.VGradient) {constB=7})) {constA=1}
  n10 = ScalarParameter() {parameterName=FixCenterLine}
  n11 = Saturate(None=Power(Base=@n8, Exp=Constant() {r=2}))
  n12 = LinearInterpolate(A=Constant(), B=Constant() {r=1}, Alpha=Divide(A=NamedRerouteUsage(), B=ScalarParameter() {defaultValue=0.5, parameterName=Distortion})) {constB=0.5}
  n13 = Constant() {r=1}
  n14 = Constant()
  n15 = Saturate(None=Add(A=LinearInterpolate(A=@n8, B=@n9, Alpha=@n10), B=If(A=@n11, B=@n12, "A > B"=@n13, "A < B"=@n14)))
  MaterialFunctionCall(Emissive=LinearInterpolate(A=@n1.RGB, B=LinearInterpolate(A=@n1.RGB, B=@n2, Alpha=@n5), Alpha=@n15)) {materialFunction=&MF_FX_EmissiveAdjustments}
  n16 = Multiply(A=Multiply(A=Multiply(A=@n15, B=@n3), B=NamedRerouteUsage()), B=ParticleColor().A)
  n17 = TwoSidedSign()
  n18 = ScalarParameter() {parameterName=Disable2SideSign}
  Multiply(
   A=MaterialFunctionCall!(TexCoord=@n6, ClampChannel=Constant2Vector() {r=1}) {materialFunction=/Game/Effects/Fort_Effects/MaterialFunctions/MF_FX_TilingMethod}.Result
   B=LinearInterpolate(A=DepthFade(Opacity=@n16), Alpha=LinearInterpolate(A=@n17, Alpha=@n18) {constB=0}) {constB=0}
  )
  n19 = MaterialFunctionCall() {materialFunction=&LinearGradient}
  n20 = AppendVector(A=Power(Base=@n19.UGradient) {constExponent=0.5}, B=@n19.VGradient)
  n21 = AppendVector(
   A=ScalarParameter() {defaultValue=3, parameterName=DistrontionU_Tile}
   B=ScalarParameter() {defaultValue=10, parameterName=DistrontionV_Tile}
  )
  n22 = Panner(Coordinate=Divide(A=Multiply(A=@n20, B=@n21), B=Constant() {r=3})) {speedX=0.2}
  n23 = Divide(A=@n22) {constB=1.5}
  NamedRerouteDeclaration(None=LinearInterpolate(
    A=TextureSample(UVs=@n22) {texture=/Game/Effects/Fort_Effects/Textures/Plasma/T_Plasma_01_M}.G
    B=TextureSample(UVs=@n23) {texture=&T_Plasma_01_M}.B
    Alpha=ScalarParameter() {defaultValue=0.1, parameterName=Alpha}
   )) {name=UV_Distortion, nodeColor=<rgba 1 0 0.0788 1>}
  n24 = DynamicParameter() {paramNames [Erosion,ThicknessH,UVOffset,Param4]}
  n25 = LinearInterpolate(Alpha=@n24.Erosion) {constA=0.2}
  NamedRerouteDeclaration(None=LinearInterpolate(
    A=OneMinus(None=@n25)
    B=MaterialFunctionCall() {materialFunction=&DebugTimeSine}.Result
    Alpha=ScalarParameter() {parameterName=DebugErosion}
   )) {name=DebugErosion, nodeColor=<rgba 0.167 0 1 1>}
  n26 = TextureCoordinate()
  n27 = Multiply(A=TextureCoordinate() {coordinateIndex=1}, B=Constant2Vector() {g=1, r=1})
  n28 = ScalarParameter() {defaultValue=1, parameterName=TexCoordChannel_0_Or_1}
  n29 = ScalarParameter() {parameterName=EnableUGrowth}
  NamedRerouteDeclaration(None=Add(A=LinearInterpolate(A=@n26, B=@n27, Alpha=@n28), B=LinearInterpolate(B=@n24.UVOffset, Alpha=@n29))) {name=TexCoord, nodeColor=<rgba 1 0 0 1>}
  n30 = MaterialFunctionCall() {materialFunction=&ObjectScale}
  n31 = ScalarParameter() {defaultValue=1, parameterName=UseScaleWidthByObjectSize}
  n32 = MaterialFunctionCall(UV Channel=TextureCoordinate() {coordinateIndex=1}) {materialFunction=&LinearGradient}
  n33 = Constant() {r=1.1}
  n34 = Multiply(A=VertexNormalWS(), B=LinearInterpolate(Alpha=Power(Base=@n32.VGradient, Exp=@n33)))
  n35 = LinearInterpolate(Alpha=Multiply(
    A=ScalarParameter() {defaultValue=1, parameterName=TicknessH}
    B=DynamicParameter() {paramNames [Erosion,ThicknessH,UVOffset,Param4]}.ThicknessH
   )) {constA=71, constB=0}
  NamedRerouteDeclaration(None=Multiply(A=LinearInterpolate(B=@n30.Scale X, Alpha=@n31) {constA=1}, B=Multiply(A=@n34, B=@n35))) {name=WPO, nodeColor=<rgba 0 0.493 1 1>}
  n36 = Constant() {r=1}
  n37 = NamedRerouteUsage()
  n38 = NamedRerouteUsage()
  n39 = SphereMask(
   A=Power(Base=MaterialFunctionCall(UV Channel=@n37) {materialFunction=&LinearGradient}.UGradient) {constExponent=0.5}
   B=AppendVector(A=LinearInterpolate(Alpha=@n38) {constA=0.5, constB=0}, B=Constant() {r=0.5})
  ) {attenuationRadius=2, hardnessPercent=0}
  n40 = Multiply(A=NamedRerouteUsage(), B=AppendVector(A=Constant() {r=0.1}, B=Constant() {r=0.1}))
  n41 = ScalarParameter() {defaultValue=0.43, parameterName=Param_10}
  n42 = AppendVector(A=Constant(), B=Constant() {r=1})
  n43 = OneMinus(None=TextureSample(UVs=Panner(Coordinate=@n40, Time=@n41, Speed=@n42) {speedY=1}) {samplerType=SAMPLERTYPE_LinearGrayscale, texture=&T_SimplePerlin_Greyscale}.R)
  n44 = Multiply(A=Multiply(A=OneMinus(None=@n39), B=Power(Base=@n43) {constExponent=1.8}), B=Constant() {r=6})
  n45 = LinearInterpolate(Alpha=NamedRerouteUsage()) {constB=2}
  n46 = Multiply(A=Subtract(A=Subtract(A=@n36, B=@n44), B=Multiply(A=@n43, B=@n45) {constB=0}), B=Constant() {r=3})
  n47 = MaterialFunctionCall(UV Channel=NamedRerouteUsage()) {materialFunction=&LinearGradient}
  n48 = LinearInterpolate(A=Multiply(A=Multiply(A=@n47.UGradient, B=@n47.UGradient)) {constB=2}, Alpha=ScalarParameter() {parameterName=DisableShrinkOnTail}) {constB=0}
  n49 = Multiply(A=Sine(None=Multiply(A=@n47.UGradient) {constB=0.5})) {constB=8}
  n50 = Subtract(A=Saturate(None=Subtract(A=@n46, B=@n48)), B=OneMinus(None=Saturate(None=@n49))) {constB=0}
  n51 = NamedRerouteUsage()
  n52 = Divide(A=NamedRerouteUsage()) {constB=15}
  n53 = NamedRerouteUsage()
  n54 = MaterialFunctionCall(UV Channel=Rotator(Coordinate=Add(A=@n51, B=@n52), Time=LinearInterpolate(Alpha=@n53)) {centerX=0, centerY=0, speed=1}) {materialFunction=&LinearGradient}
  n55 = Saturate(None=Saturate(None=Multiply(A=@n54.UGradient) {constB=15}))
  n56 = ScalarParameter() {parameterName=UseSharpenTrailHead}
  NamedRerouteDeclaration(None=Multiply(A=Saturate(None=@n50), B=LinearInterpolate(B=@n55, Alpha=@n56) {constA=1})) {name=Dissolve, nodeColor=<rgba 0.6953 0.0318 1 1>}
  NamedRerouteUsage()
  ScalarParameter()
  VertexTangentWS()
 ]
}
